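# Idle Niagara effect for the HollyDessert pickaxe that plays short-lived purple lightning meshes.
# Its single Lightning emitter feeds two mesh renderers from two spawn groups with different rotations, using probabilistic staggered bursts plus slow continuous spawning.
# A per-particle waveform flickers alpha, glow and one dynamic material parameter, while curves drive mesh scale and other material inputs.
# A system time-of-day scalar dims or boosts color, and front-end modifiers set separate lobby and gameplay intensities.
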NiagaraSystem /BRCosmetics/Weapons/FORT_Melee/Pickaxe_HollyDessert/FX/NS_Pickaxe_HollyDessert_Idle
dependencies {
 usedDataInterfaces [&NiagaraDataInterfaceVectorCurve, &NiagaraDataInterfaceCurve]
 usedDynamicInputs [&FNTimeOfDayScalar]
 usedModules [
  &SystemState
  &self:SystemSpawnScript.SystemScriptSource.SystemScriptGraph.NiagaraNodeAssignment_1.SetVariables_#1
 ]
}
emitters {Lightning {
  emitterUpdateScript [
   &EmitterState()
   &SpawnBurst_Instantaneous(Spawn Probability=0.6, Use Spawn Probability=true, UseLoopCountLimit=true)
   &SpawnBurst_Instantaneous(Spawn Group=1, Spawn Probability=0.6, Spawn Time=0.08, Use Spawn Probability=true, UseLoopCountLimit=true) {module=SpawnBurst_Instantaneous001}
   &SpawnRate(SpawnRate=&RandomRangeFloat(Minimum=0.8, Maximum=1.4), Use Spawn Probability=false) {module=SpawnRate001}
   &SpawnRate(SpawnGroup=1, SpawnRate=&RandomRangeFloat(Minimum=0.8, Maximum=1.4), Use Spawn Probability=false) {module=SpawnRate002}
  ]
  particleSpawnScript [
   &InitializeParticle(
    Color=<rgba 1.4605 0.8542 4 1>
    Lifetime=&RandomRangeFloat(Minimum=0.6, Maximum=0.7)
    Mesh Scale=&RandomRangeVector(Minimum=<1.37 1 0.45>, Maximum=<1.45 1 0.55>)
    Mesh Scale Mode=::NewEnumerator4
    Position Offset=&RandomRangeVector(Minimum=<0 0 10>, Maximum=<0 0 13>)
    UsePositionOffset=true
    Write Material Random=true
   )
   &InitialMeshOrientation(
    Mesh Orientation Mode=::NewEnumerator2
    Orientation Vector=<0 0 -1>
    Rotation=&MaskVectorBySpawnGroup(
     Masked Vector=&RandomRangeVector(Minimum=<-0.05 0 0>, Maximum=<0.05 0 0>)
     Passthrough Vector=&RandomRangeVector(Minimum=<0.45 0 0>, Maximum=<0.55 0 0>)
    )
    Use Rotation Vector=true
   )
   &CameraOffset(Camera Offset Amount=1)
  ]
  particleUpdateScript [
   &ParticleState()
   /SetVariables(Particles.WaveFormFloat:NiagaraFloat=&Waveform(
     Global Input Position=&RandomRangeFloat(Minimum=0, Maximum=1)
     Global Phase=$Particles.Age
     "Waveform [1]"=::NewEnumerator4
     "[1] Phase Scale"=4.3463
     "[1] Phase"=$Particles.Age
     "[1] Frequency"=1.3
     "[1] Amplitude Min/Max"=<0 1>
     "[1] Bias"=0.376
    )) {isSetParameters=true}
   &ScaleMeshSize(Scale Factor=&VectorFromCurve(VectorCurve=DataInterface() {propertyValues {
       XCurve {keys [{arriveTangent=1.0483, interpMode=RCIM_Cubic, time=0, value=1}]}
       YCurve {keys [<at 0 1>,<at 1 1>]}
       ZCurve {keys [<at 0 1 RCIM_Cubic>,<at 1 0.6974 RCIM_Cubic>]}
      }}))
   &DynamicMaterialParameters(
    Index 0 Param 1=&FloatFromCurve(FloatCurve=DataInterface() {propertyValues {Curve {keys [
         {arriveTangent=0.6704, interpMode=RCIM_Cubic, leaveTangent=0.6704, tangentMode=RCTM_User, time=0, value=0}
         {arriveTangent=1.2509, interpMode=RCIM_Cubic, leaveTangent=1.2509, tangentMode=RCTM_User, time=0.7447, value=0.7331}
         <at 1 1 RCIM_Cubic>
        ]}}})
    Index 0 Param 2=0.7
    Index 0 Param 3=&FloatFromCurve(FloatCurve=DataInterface() {propertyValues {Curve {keys [
         {arriveTangent=-0.111, interpMode=RCIM_Cubic, time=0, value=1}
         {arriveTangent=-0.6227, interpMode=RCIM_Cubic, leaveTangent=-0.6227, tangentMode=RCTM_User, time=1, value=0.6925}
        ]}}})
    Index 0 Param 4=0.1
    Index 1 Param 1=&FloatFromCurve(
     FloatCurve=DataInterface() {propertyValues {Curve {keys [{arriveTangent=-0.7, interpMode=RCIM_Cubic, time=0.1113, value=0.5744}, <at 1 0.3785>]}}}
     Scale Curve=1.7
    )
    Index 1 Param 2=&Add_Float(
     A=&RandomRangeFloat(Minimum=0, Maximum=1)
     B=&Add_Float(
      A=&FloatFromCurve(
       FloatCurve=DataInterface() {propertyValues {Curve {keys [{arriveTangent=-2.9709, leaveTangentWeight=0.0667, time=0, value=1}, {arriveTangent=1.0367, arriveTangentWeight=0.0667, time=1, value=0}]}}}
       Scale Curve=&RandomRangeFloat(Minimum=0.12, Maximum=0.23)
      )
      B=&Multiply_Float(A=$Particles.WaveFormFloat, B=&RandomRangeFloat(Minimum=0.3, Maximum=0.7))
     )
    )
    Index 1 Param 3=2
    Index 1 Param 4=&FloatFromCurve(FloatCurve=DataInterface() {propertyValues {Curve {keys [
         {arriveTangent=-1.4071, interpMode=RCIM_Cubic, leaveTangent=-1.4071, tangentMode=RCTM_User, time=0.0928, value=1}
         {arriveTangent=-0.118, tangentMode=RCTM_User, time=1, value=0.0986}
        ]}}})
    Module.Index0Float0Write=true
    Module.Index0Float1Write=true
    Module.Index0Float2Write=true
    Module.Index0Float3Write=true
    Module.Index1Float0Write=true
    Module.Index1Float1Write=true
    Module.Index1Float2Write=true
    Module.Index1Float3Write=true
    Write Parameter Index 1=::NewEnumerator1
   )
   /SetVariables(Particles.Glow Color=&MakeLinearColorFromVectorAndFloat(
     "Vector (RGB)"=&MakeVectorFromLinearColorRGB(Color=&MultiplyLinearColor(Linear Color=<rgba 0.4727 0.2767 1 1>, Scale Factor=&MakeVector4_FromFloat(Float=$System.TOD)))
     "Float (Alpha)"=&Multiply_Float(
      A=&FloatFromCurve(
       FloatCurve=DataInterface() {propertyValues {Curve {keys [
           {tangentMode=RCTM_User, time=0, value=1}
           {arriveTangent=-0.0405, interpMode=RCIM_Cubic, leaveTangent=-0.0405, tangentMode=RCTM_User, time=0.3017, value=1}
           {arriveTangent=-0.1459, interpMode=RCIM_Cubic, leaveTangent=-0.1459, tangentMode=RCTM_User, time=1, value=0}
          ]}}}
       Scale Curve=0.25
      )
      B=&Multiply_Float(A=&FN_FrontEndFloatModifier(Gameplay=11, Front End=8), B=$Particles.WaveFormFloat)
     )
    )) {isSetParameters=true}
   /SetVariables(Particles.Glow Dy02:Vector4f=&MakeVector4(
     X=&FloatFromCurve(FloatCurve=DataInterface() {propertyValues {Curve {keys [
          {arriveTangent=0.0616, interpMode=RCIM_Cubic, leaveTangent=0.0616, tangentMode=RCTM_User, time=0, value=0}
          {interpMode=RCIM_Cubic, leaveTangent=1.8478, time=1, value=1}
         ]}}})
     Y=0.7
     Z=0
     W=0.1
    )) {isSetParameters=true}
   &ScaleColor(
    Scale Alpha=&Multiply_Float(
     A=&FN_FrontEndFloatModifier(Gameplay=15, Front End=10)
     B=&FloatFromCurve(
      FloatCurve=DataInterface() {propertyValues {Curve {keys [
          {tangentMode=RCTM_User, time=0, value=1}
          {arriveTangent=-0.0405, interpMode=RCIM_Cubic, leaveTangent=-0.0405, tangentMode=RCTM_User, time=0.466, value=1}
          {arriveTangent=-0.1459, interpMode=RCIM_Cubic, leaveTangent=-0.1459, tangentMode=RCTM_User, time=1, value=0}
         ]}}}
      Scale Curve=&Multiply_Float(A=0.65, B=$Particles.WaveFormFloat)
     )
    )
    Scale RGB=&Multiply_VectorByFloat(
     Vector=<1 1 1>
     Float=&Multiply_Float(A=1, B=&Multiply_Float(A=1, B=$System.TOD))
    )
    ScaleA=true
    ScaleRGB=true
   )
  ]
  rendererClasses [&NiagaraMeshRendererProperties]
  renderers [
   {rendererClass=&NiagaraMeshRendererProperties, rendererIndex=0}
   {rendererClass=&NiagaraMeshRendererProperties, rendererIndex=1}
  ]
 }}
summary {
 systemName=NS_Pickaxe_HollyDessert_Idle
 emitters [{emitterName=Lightning, rendererClasses [&NiagaraMeshRendererProperties]}]
 userVariables [
  User.bIsFrontEnd {type=NiagaraBool}
  User.bIsFrontEndPreview {type=NiagaraBool}
 ]
}
