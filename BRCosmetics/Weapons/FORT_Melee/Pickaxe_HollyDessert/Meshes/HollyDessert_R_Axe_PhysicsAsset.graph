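# Physics asset for the right-hand axe mesh of the HollyDessert pickaxe cosmetic melee weapon.
# It holds a single capsule body on the Melee_TwoHanded_Handle bone, offset along and rotated to follow the handle, giving the weapon a simple collision shape.
# With only one body, it has no constraints.
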PhysicsAsset /BRCosmetics/Weapons/FORT_Melee/Pickaxe_HollyDessert/Meshes/HollyDessert_R_Axe_PhysicsAsset
bodies {Melee_TwoHanded_Handle [{shapeType=Capsule, length=31.3885, radius=3.7725, center=<-0.4212 -0.3053 17.9991>, rotation=<rot -0.0052 -90 0.0084>}]}
constraints []
